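; Controller component for roll landing that fires a tutorial event and computes roll stamina gain.
; After a delayed begin play, it binds to the current pawn or, lacking one, to the player controller's pawn-change event.
; For local controllers with tutorials enabled, an async task watches for the can-activate tag and sends the tutorial event.
; UpdateAndGetStaminaRegen scales gain by a fatigue index or multiplier, resetting fatigue after a timeout.
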(event EventBeginPlay
  (bind _returnvalue (CallFunction|GetPawn))
  (bind _asfort_player_controller (CallFunction|GetOwnerController))
  (Utilities|FlowControl|Delay 1.0)
  (Utilities|FlowControl|Sequence
    (:then_0
      (Utilities|IsValid _returnvalue
        (:Is%20Valid
          (bind _asfort_player_pawn (Utilities|Casting|CastToFortPlayerPawn _returnvalue))
          (CallFunction|BindRollingEvents _asfort_player_pawn))
        (:Is%20Not%20Valid
          (Utilities|IsValid _asfort_player_controller
            (:Is%20Valid
              (EventDispatchers|AssignOnFortPawnChangedEvent _asfort_player_controller (delegate Custom|OnFortPawnChangedEvent_Event)))) ; If not Valid, this is a PlayerBot or AI! Right now do nothing, but we might want to have Bots do something here at one point.
        )))))

(event EventEndPlay
  (params (EndPlayReason :type EEndPlayReason%20Enum :default "Destroyed"))
  (|CallParentFunction EndPlayReason)
  (CallFunction|EndTasks))

(event Custom|OnFortPawnChangedEvent_Event
  (params (NewPawn :type Fort%20Pawn%20Object%20Reference))
  (CallFunction|BindRollingEvents NewPawn))

(event Custom|BindRollingEvents
  (params (Target :type Character%20Object%20Reference))
  (CallFunction|EndTasks)
  ; Bind activatable tag add for local player only, since it's for tutorials
  (if (ScalableFloat|GetValueasBool (Variables|Default|GetTutorialEnabled))
    (if (Pawn|IsLocalController (CallFunction|GetOwnerController))
      (bind _asynctaskproxy (Ability|Tasks|WaitGameplayTagAddToActor Target (Variables|Default|GetCanActivateTag))
        (:then
          (Variables|Default|SetWaitTagTask _asynctaskproxy))
        (:Added
          (Ability|SendGameplayEventtoActor (CallFunction|GetPawn) (Variables|Default|GetRollLandingTutorialEventTag)))))))

(fn EndTasks
  (bind _waittagtask (Variables|Default|GetWaitTagTask)
    (:then
      (Ability|Async|EndAction _waittagtask))
    (:else)))

(fn GetOwnerController
  (bind _asfort_player_controller (Utilities|Casting|CastToFortPlayerController (Components|GetOwner))
    (:then
      _)
    (:CastFailed
      _))
  (return _asfort_player_controller))

(fn GetPawn
  (bind _asfort_player_controller (CallFunction|GetOwnerController))
  (Utilities|IsValid _asfort_player_controller
    (:Is%20Valid
      (return (Pawn|GetControlledPawn _asfort_player_controller)))
    (:Is%20Not%20Valid
      (bind _asfort_bot_controller (Utilities|Casting|CastToFortBotController (Components|GetOwner)))
      (return (Pawn|GetControlledPawn _asfort_bot_controller)))))

(fn UpdateAndGetStaminaRegen
  (bind _fatigueindexthistime (LocalVariables|GetFatigueIndexThisTime))
  (bind _returnvalue (Math|Conversions|ToFloat(Integer) _fatigueindexthistime))
  (bind _staminagainfatigue (Variables|Default|GetStaminaGainFatigue))
  (bind _returnvalue_2 (ScalableFloat|GetValueatLevelwithContext _staminagainfatigue))
  (bind _staminagainamount (Variables|Default|GetStaminaGainAmount))
  (bind _returnvalue_3 (ScalableFloat|GetValueatLevelwithContext _staminagainamount _returnvalue))
  (Utilities|FlowControl|Sequence
    (:then_0
      (if (or (< (Variables|Default|GetTimeLastRoll) 0) (> (- (Utilities|Time|GetTimeSeconds) (Variables|Default|GetTimeLastRoll)) (ScalableFloat|GetValueatLevel (Variables|Default|GetStaminaFatigueResetTime))))
        (bind _output_get (Variables|Default|SetStaminaFatigueIndex 0))
        (LocalVariables|SetFatigueIndexThisTime _output_get)
        (else
          (LocalVariables|SetFatigueIndexThisTime (Variables|Default|GetStaminaFatigueIndex)))))
    (:then_1
      (Variables|Default|SetTimeLastRoll (Utilities|Time|GetTimeSeconds))
      (Math|Integer|IncrementInt (Variables|Default|GetStaminaFatigueIndex)
        (:%20%20
          ; Stamina Gain Fatigue is a multiplier on the previously applied Stamina Gain Amount. Check if it's valid and we've already bounced before using it
          (if (> _returnvalue_2 0)
            (if (> (LocalVariables|GetFatigueIndexThisTime) 0)
              (bind _output_get_2 (Variables|Default|SetStaminaAmountLastRoll (* (Variables|Default|GetStaminaAmountLastRoll) _returnvalue_2)))
              (return _output_get_2)
              (else
                _))
            (else
              _))
          (Variables|Default|SetStaminaAmountLastRoll _returnvalue_3)
          (return _returnvalue_3))))))
